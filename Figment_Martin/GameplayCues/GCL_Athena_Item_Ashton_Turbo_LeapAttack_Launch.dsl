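; Gameplay cue for the Turbo item's leap attack launch that spawns a takeoff particle effect on the ground beneath the pawn.
; On application it stores the target pawn and traces downward by a height-check variable, spawning an effect only if ground is hit.
; The emitter appears at the hit location with the captured actor rotation, using a separate local variant for locally controlled pawns.
; It then calls the parent application handler.
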(event EventOnApplication(Cascade)
  (params (MyTarget :type Actor%20Object%20Reference) (Parameters :type Gameplay%20Cue%20Parameters%20Structure%20%28by%20ref%29) (ParticleComponents :type Array%20of%20Cascade%20Particle%20System%20Component%20Object%20References) (AudioComponents :type Array%20of%20Audio%20Component%20Object%20References) (BurstCameraShakeInstance :type Camera%20Shake%20Base%20Object%20Reference) (BurstDecalInstance :type Decal%20Actor%20Object%20Reference))
  (bind _returnvalue (Transformation|GetActorRotation))
  ;[1 1 1] Only Spawn The FX if were are close to the ground and spawn them on the ground. Check if local and a local fx if true
  (bind _asfort_player_pawn (Utilities|Casting|CastToFortPlayerPawn MyTarget))
  (Variables|Default|SetTargetPawn _asfort_player_pawn)
  (bind (_outhit _returnvalue_2) (Collision|LineTraceByChannel 0 (Transformation|GetActorLocation) (Math|Vector|vector-vector (Transformation|GetActorLocation) (Math|Vector|MakeVector 0.0 0.0 (Variables|Default|GetHeightCheck))) "TraceTypeQuery10"))
  (if _returnvalue_2
    (if (Pawn|IsLocallyControlled (Variables|Default|GetTargetPawn))
      (Effects|SpawnEmitteratLocation 0 "/Figment_Martin/Gameplay/Martin_Items/Turbo/FX/P_Ashton_Turbo_JumpAttack_StartTakeoff_Local.P_Ashton_Turbo_JumpAttack_StartTakeoff_Local" (.Location (Collision|BreakHitResult _outhit)) _returnvalue)
      (else
        (Effects|SpawnEmitteratLocation 0 "/Figment_Martin/Gameplay/Martin_Items/Turbo/FX/P_Ashton_Turbo_JumpAttack_StartTakeoff.P_Ashton_Turbo_JumpAttack_StartTakeoff" (.Location (Collision|BreakHitResult _outhit)) _returnvalue)))
    (else
      _))
  (|CallParentFunction 0 Parameters ParticleComponents AudioComponents BurstCameraShakeInstance BurstDecalInstance))

(fn ConstructionScript)
