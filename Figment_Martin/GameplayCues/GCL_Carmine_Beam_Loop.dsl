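; Looping gameplay cue for the Carmine backpack beam, handling its dust particles and audio on the target player pawn.
; On becoming relevant, it destroys any previous dust effect, attaches and stores a new laser-blast dust emitter at the mesh root, and spawns the beam sound at the audio component's location.
; Particle deactivation stops the stored dust.
; Ceasing relevance fades out the cue's audio component, not the separately spawned sound.
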(event GameplayCueNotify|EventHandleGameplayCue
  (params (MyTarget :type Actor%20Object%20Reference) (EventType :type EGameplayCueEvent%20Enum :default "OnActive") (Parameters :type Gameplay%20Cue%20Parameters%20Structure%20%28by%20ref%29)))

(event EventBeginPlay)

(event EventEndPlay
  (params (EndPlayReason :type EEndPlayReason%20Enum :default "Destroyed")))

(event Collision|EventActorBeginOverlap
  (params (OtherActor :type Actor%20Object%20Reference)))

(event EventTick
  (params (DeltaSeconds :type Float%20%28single-precision%29)))

(fn OnCeaseRelevant
  (params (MyTarget :type Actor%20Object%20Reference) (Parameters :type Gameplay%20Cue%20Parameters%20Structure%20%28by%20ref%29))
  (bind _audiocomponent (Variables|Sounds|GetAudioComponent)
    (:then
      (Audio|Components|Audio|FadeOut _audiocomponent 0.5 0.0 "Linear" :!class /Script/Engine.AudioComponent)
      (return false))
    (:else)))

(fn OnStopParticleSystemSpawned
  (params (SpawnedParticleSysComponent :type Cascade%20Particle%20System%20Component%20Object%20Reference) (Parameters :type Gameplay%20Cue%20Parameters%20Structure%20%28by%20ref%29))
  (return))

(fn OnWhileActiveParticleSystemActivate
  (params (WhileActiveParticleSysComponent :type Cascade%20Particle%20System%20Component%20Object%20Reference) (Parameters :type Gameplay%20Cue%20Parameters%20Structure%20%28by%20ref%29))
  (return))

(fn OnWhileActiveParticleSystemDeactivate
  (params (WhileActiveParticleSysComponent :type Cascade%20Particle%20System%20Component%20Object%20Reference) (Parameters :type Gameplay%20Cue%20Parameters%20Structure%20%28by%20ref%29))
  (Components|Activation|Deactivate (Variables|Default|GetDustFX))
  (return))

(fn ConstructionScript)

(fn OnBecomeRelevant
  (params (MyTarget :type Actor%20Object%20Reference) (Parameters :type Gameplay%20Cue%20Parameters%20Structure%20%28by%20ref%29))
  (bind _asplayer_pawn_athena (Utilities|Casting|CastToPlayerPawn_Athena MyTarget))
  (bind _dustfx (Variables|Default|GetDustFX)
    (:then
      (Components|DestroyComponent _dustfx 0))
    (:else
      _))
  (bind _returnvalue (Effects|SpawnEmitterAttached :EmitterTemplate "/Figment_Martin/Athena/Items/Gameplay/BackPacks/CarminePack/FX/P_Jim_LaserBlast_Dust.P_Jim_LaserBlast_Dust" :AttachToComponent (Variables|/Script/Engine.Character|GetMesh _asplayer_pawn_athena) :AttachPointName "root" :Rotation "0,90.000000, 0" :Scale "1.000000,1.000000,1.000000" :LocationType "SnapToTarget" :bAutoDestroy true :PoolingMethod "None" :bAutoActivate true))
  (Variables|Default|SetDustFX _returnvalue)
  (Audio|SpawnSoundatLocation 0 "/Figment_Martin/Sounds/Fort_GamePlay_Sounds/Carmine/Carmine_Beam_StartLoopStop_Cue.Carmine_Beam_StartLoopStop_Cue" (Transformation|GetWorldLocation (Variables|Sounds|GetAudioComponent)))
  (return false))
