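; Gameplay cue that plays the Carmine backpack's laser beam effects when executed, with force feedback at the target actor.
; It spawns the beam from the player's Hand_L socket toward the cue location, setting its scale and life from distance over MaxBeamLength.
; Rings and Glow emitters are disabled unless the instigator is locally controlled.
; A normalized magnitude of 1 adds impact effects, choosing player or world by casting EffectCauser to FortPawn, with first-person sounds for local players.
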(fn OnExecute
  (params (MyTarget :type Actor%20Object%20Reference) (Parameters :type Gameplay%20Cue%20Parameters%20Structure%20%28by%20ref%29))
  (bind (_normalizedmagnitude _rawmagnitude _effectcontext _matchedtagname _originaltag _aggregatedsourcetags _aggregatedtargettags _location _normal _instigator _effectcauser) (Ability|GameplayCue|BreakGameplayCueParameters Parameters))
  (bind _returnvalue (Math|Conversions|RotationFromXVector _normal))
  (bind _maxbeamlength (Variables|Default|GetMaxBeamLength))
  (Fortnite|PlayLocalForceFeedbackatLocation 0 "/Game/Weapons/ForceFeedback/FF_Fire_Med_R.FF_Fire_Med_R" (Transformation|GetActorLocation MyTarget) 2e+02)
  (bind _asplayer_pawn_athena (Utilities|Casting|CastToPlayerPawn_Athena MyTarget))
  (Audio|SpawnSoundatLocation 0 0 (Transformation|GetSocketLocation (Variables|/Script/Engine.Character|GetMesh _asplayer_pawn_athena) "Hand_L"))
  (bind _returnvalue_2 (Effects|SpawnEmitteratLocation 0 "/Figment_Martin/Athena/Items/Gameplay/BackPacks/CarminePack/FX/P_Jim_LaserBlast.P_Jim_LaserBlast" (Transformation|GetSocketLocation (Variables|/Script/Engine.Character|GetMesh _asplayer_pawn_athena) "Hand_L") (Math|Rotator|FindLookatRotation (Transformation|GetSocketLocation (Variables|/Script/Engine.Character|GetMesh _asplayer_pawn_athena) "Hand_L") _location) (1.0 1.0 1.0) true "None" true))
  (Effects|Components|ParticleSystem|SetVectorParameter _returnvalue_2 "BeamScale" (Math|Vector|MakeVector (/ (Math|Vector|VectorLength (- (Transformation|GetSocketLocation (Variables|/Script/Engine.Character|GetMesh _asplayer_pawn_athena) "Hand_L") _location)) _maxbeamlength) 1.0 1.0))
  (Effects|Components|ParticleSystem|SetFloatParameter _returnvalue_2 "BeamLife" (/ (Math|Vector|VectorLength (- (Transformation|GetSocketLocation (Variables|/Script/Engine.Character|GetMesh _asplayer_pawn_athena) "Hand_L") _location)) _maxbeamlength))
  ;[1 1 1] Shuts off Rings for Not player controlled Jim
  (if (not (Ability|GameplayCue|IsInstigatorLocallyControlled Parameters))
    (Effects|Components|ParticleSystem|SetEmitterEnable _returnvalue_2 "Rings" false)
    (Effects|Components|ParticleSystem|SetEmitterEnable _returnvalue_2 "Glow" false))
  (if (== _normalizedmagnitude 1.0)
    (Utilities|Casting|CastToFortPawn _effectcauser
      (:then
        ;[1 1 1] Hit PlayerFX
        (Effects|SpawnEmitteratLocation 0 "/Figment_Martin/Athena/Items/Gameplay/BackPacks/CarminePack/FX/P_Jim_LaserBlast_Impact_Player.P_Jim_LaserBlast_Impact_Player" _location _returnvalue)
        (Audio|PlaySoundatLocation 0 (select (Pawn|IsLocallyControlled _asplayer_pawn_athena) "/Figment_Martin/Sounds/Fort_GamePlay_Sounds/Carmine/Carmine_Beam_Impact_Player_1P_Cue.Carmine_Beam_Impact_Player_1P_Cue" "/Figment_Martin/Sounds/Fort_GamePlay_Sounds/Carmine/Carmine_Beam_Impact_Player_Cue.Carmine_Beam_Impact_Player_Cue") _location))
      (:CastFailed
        ;[1 1 1] Hit WorldFX
        (Effects|SpawnEmitteratLocation 0 "/Figment_Martin/Athena/Items/Gameplay/BackPacks/CarminePack/FX/P_Jim_LaserBlast_Impact.P_Jim_LaserBlast_Impact" _location _returnvalue)
        (Audio|PlaySoundatLocation 0 (select (Pawn|IsLocallyControlled _asplayer_pawn_athena) "/Figment_Martin/Sounds/Fort_GamePlay_Sounds/Carmine/Carmine_Beam_Impact_World_1P_Cue.Carmine_Beam_Impact_World_1P_Cue" "/Figment_Martin/Sounds/Fort_GamePlay_Sounds/Carmine/Carmine_Beam_Impact_World_Cue.Carmine_Beam_Impact_World_Cue") _location)))
    (return (|CallParentFunction MyTarget Parameters))))
